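# M_Cream is a masked, lit particle sprite material for cream effects on the birthday cake prop.
# Per-particle random values and world position offset repeatable noise UVs, varying a distortion normal blended over a sphere normal for rounded, uneven blobs.
# The same noise drives a sphere mask radius that shades the particle color, feeding a brightness-scaled base color and switchable emissive with a night tint.
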Material /Game/Environments/Props/BirthdayCake/FX/M_Cream
graph {
 properties {
  bUsedWithNiagaraSprites=true
  bUsedWithParticleSprites=true
  blendMode=BLEND_Masked
  pixelDepthOffsetMode=PDOM_Legacy
  usedShadingModels=Default Lit
 }
 nodes [
  Constant() {r=0.5}
  Constant() {r=1}
  n1 = TextureSample() {samplerType=SAMPLERTYPE_Normal, texture=&T_Sphere_N}
  n2 = Constant3Vector() {constant=<rgba 1.2 1.2 1 1>}
  n3 = Add(A=TextureCoordinate(), B=ShadingPathSwitch(Default=ParticleRandom(), Mobile=Constant()) {desc=Migrated from FeatureLevelSwitch})
  n4 = DynamicParameter() {paramNames [Random,Param2,Param3,Param4]}
  n5 = ComponentMask(None=ParticlePositionWS()) {b=true}
  n6 = ScalarParameter() {defaultValue=400, parameterName=PannerUnit}
  n7 = Multiply(
   A=Add(A=Add(A=@n3, B=@n4.Random), B=Divide(A=@n5, B=@n6) {constB=400})
   B=AppendVector(A=ScalarParameter() {defaultValue=0.75, parameterName=RepeatX}, B=ScalarParameter() {defaultValue=0.75, parameterName=RepeatY})
  )
  n8 = TextureSample(UVs=@n7) {samplerType=SAMPLERTYPE_Normal, texture=&T_Weap_Distort_N_02}
  n9 = Constant3Vector() {constant=<rgba 1 1 1 1>}
  MaterialFunctionCall(BaseNormal=Normalize(VectorInput=Multiply(A=@n1.RGB, B=@n2)), AdditionalNormal=Normalize(VectorInput=Multiply(A=@n8.RGB, B=@n9))) {materialFunction=&BlendAngleCorrectedNormals}
  n10 = TextureCoordinate()
  n11 = Constant() {r=0.5}
  n12 = LinearInterpolate(Alpha=TextureSample(UVs=@n7) {samplerType=SAMPLERTYPE_Grayscale, texture=&LowResBlurredNoise}.R) {constA=0.2, constB=0.6}
  n13 = Constant()
  n14 = Multiply(
   A=ParticleColor().RGB
   B=LinearInterpolate(A=ScalarParameter() {defaultValue=0.75, parameterName=Shadows}, Alpha=SphereMask(A=@n10, B=@n11, Radius=@n12, Hardness=@n13)) {constA=0.75}
  )
  n15 = Multiply(A=@n14, B=ScalarParameter() {parameterName=EmissiveAmount})
  n16 = Constant() {r=1}
  n17 = MaterialFunctionCall(
   Color_Morning=@n16
   Color_Day=@n16
   Color_Evening=@n16
   Color_Night=VectorParameter() {parameterName=NightColorTweak, defaultValue=<rgba 1 1 1 0>}.RGB
  ) {materialFunction=&Color_Time_of_Day}
  MaterialFunctionCall(Emissive=StaticSwitchParameter(True=Multiply(A=@n15, B=@n17.Result), False=Constant()) {parameterName=UseEmissive}) {materialFunction=&MF_FX_EmissiveAdjustments}
  Multiply(A=@n14, B=ScalarParameter() {defaultValue=1, parameterName=BaseColorBrightness})
 ]
}
